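; Non-participant AI pawn for the VividRazor CoffeePuff boss NPC, holding its energy, facing, location memory and active hookshot reference.
; BeginPlay overrides energy regen with the Regroup tag; full drain tags the AI controller as EnergyDepleted.
; Spawn and last-known target locations are nav-projected into the blackboard and a location list, which trims oldest entries; bulk additions require leash and height checks.
; Look-at disables movement while a timeline lerps yaw.
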(event EventBeginPlay
  (bind _returnvalue (Actor|GetComponentbyClass "/Script/EnergyRuntime.FortComponent_Energy"))
  (|CallParentFunction)
  (Events|BindEventtoOnEnergyCompletelyDrained _returnvalue (EventDispatchers|CreateEvent :!event Custom|OnEnergyCompletelyDrained))
  (Energy|ApplyRegenDataOverride _returnvalue (GameplayTags|MakeLiteralGameplayTag "(TagName=\"Ability.VividRazor.Boss.CoffeePuff.Regroup\")") 0.0 -1.0 -1.0)
  (Utilities|Array|Clear (Variables|Default|GetAllKnownValidLocations)))

(event Collision|EventActorBeginOverlap
  (params (OtherActor :type Actor%20Object%20Reference))
  (|CallParentFunction OtherActor) ; This node is disabled and will not be called.\nDrag off pins to build functionality.
)

(event EventTick
  (params (DeltaSeconds :type Float%20%28single-precision%29))
  (|CallParentFunction DeltaSeconds))

(event Custom|OnEnergyCompletelyDrained
  (params (EnergyComponent :type Fort%20Component%20Energy%20Object%20Reference))
  (bind _asfort_athena_aibot_controller (Utilities|Casting|CastToFortAthenaAIBotController (Pawn|GetController)))
  (Ability|GameplayEffect|AddLooseGameplayTags _asfort_athena_aibot_controller (GameplayTags|MakeGameplayTagContainerfromArray (Utilities|Array|MakeArray "(TagName=\"Event.Ability.VividRazor.CoffeePuff.EnergyDepleted\")"))))

(event Custom|StartLookingAt
  (params (LookAtLocation :type Vector))
  (bind (_returnvalue_roll _returnvalue_pitch _returnvalue_yaw) (Math|Rotator|BreakRotator (Math|Rotator|FindLookatRotation (Transformation|GetActorLocation) LookAtLocation)))
  (bind (_lookatstartingrotation_roll _lookatstartingrotation_pitch _lookatstartingrotation_yaw) (Math|Rotator|BreakRotator (Variables|GetLookAtStartingRotation_Roll)))
  (bind _charactermovement (Variables|Character|GetCharacterMovement))
  (Variables|Default|SetLookAtStartingRotation (Transformation|GetActorRotation))
  (Pawn|Components|CharacterMovement|SetMovementMode _charactermovement "MOVE_None" 0)
  (bind (_direction _alpha) (Timeline|RotateTowardsTarget PlayFromStart)
    (:Update
      (Transformation|SetActorRotation (Math|Rotator|Lerp(Rotator) (Math|Rotator|MakeRotator 0.0 0.0 _lookatstartingrotation_yaw) (Math|Rotator|MakeRotator 0.0 0.0 _returnvalue_yaw) _alpha false)))
    (:Finished
      (Pawn|Components|CharacterMovement|SetMovementMode _charactermovement "MOVE_Walking" 0))))

(event Custom|StopLookingAt
  (bind (_lookatstartingrotation_roll _lookatstartingrotation_pitch _lookatstartingrotation_yaw) (Math|Rotator|BreakRotator (Variables|GetLookAtStartingRotation_Roll)))
  (bind _charactermovement (Variables|Character|GetCharacterMovement))
  (Pawn|Components|CharacterMovement|SetMovementMode _charactermovement "MOVE_Walking" 0)
  (Timeline|RotateTowardsTarget Stop))

(fn AddListOfKnownValidLocations
  (params (NewValidLocations :type Array%20of%20Vectors))
  (bind _asfort_athena_aibot_controller (Utilities|Casting|CastToFortAthenaAIBotController (Pawn|GetController)))
  (for _array_element NewValidLocations
    (bind (_returnvalue_x _returnvalue_y _returnvalue_z) (Math|Vector|BreakVector (AI|Components|Blackboard|GetValueasVector (AI|GetBlackboard _asfort_athena_aibot_controller) (Utilities|Name|MakeLiteralName "CoffeePuff_SpawnLocation"))))
    (if (and (Class|FortAthenaAIBotController|IsLocationInsideLeash _asfort_athena_aibot_controller _array_element false 0.0) (< (Math|Float|Absolute(Float) (- (.z _array_element) _returnvalue_z)) (ScalableFloat|GetValueatLevel (Variables|Default|GetMaxVerticalDistanceFromSpawn))))
      (Utilities|Array|Add (Variables|Default|GetAllKnownValidLocations) _array_element)))
  (while (and (> (Utilities|Array|Length (Variables|Default|GetAllKnownValidLocations)) (ScalableFloat|GetValueatLevel (Variables|Default|GetMaxKnownValidLocations))) (> (Utilities|Array|Length (Variables|Default|GetAllKnownValidLocations)) 1))
    (Utilities|Array|RemoveIndex (Variables|Default|GetAllKnownValidLocations))))

(fn ClearActiveHookshot
  (Variables|Default|SetActiveHookshot 0))

(fn GetActiveHookshot
  (return (Variables|Default|GetActiveHookshot)))

(fn GetAllKnownValidLocations
  (return (Variables|Default|GetAllKnownValidLocations)))

(fn HasValidLastKnownTargetLocation
  (return (Variables|Default|GetHaveValidLastKnownTargetLocation)))

(fn SetActiveHookshot
  (params (Hookshot :type B%20Prj%20Hookshot%20Object%20Reference))
  (Variables|Default|SetActiveHookshot Hookshot))

(fn SetSpawnLocation
  (params (Location :type Vector))
  (bind _lknavprobe_extentz (Variables|Default|GetLKNavProbe_ExtentZ))
  (bind _returnvalue (ScalableFloat|GetValueatLevel _lknavprobe_extentz))
  (bind _lknavprobe_extenty (Variables|Default|GetLKNavProbe_ExtentY))
  (bind _returnvalue_2 (ScalableFloat|GetValueatLevel _lknavprobe_extenty))
  (bind _lknavprobe_extentx (Variables|Default|GetLKNavProbe_ExtentX))
  (bind _returnvalue_3 (ScalableFloat|GetValueatLevel _lknavprobe_extentx))
  (bind _self self)
  (bind _returnvalue_4 (Fortnite|AI|GetNavigationDataforActor _self))
  (bind (_projectedlocation _returnvalue_5) (AI|Navigation|ProjectPointtoNavigation Location _returnvalue_4 "/Game/Athena/AI/Phoebe/NavFilter/NavFilter_Phoebe.NavFilter_Phoebe_C" (Math|Vector|MakeVector _returnvalue_3 _returnvalue_2 _returnvalue)))
  (if _returnvalue_5
    (AI|Components|Blackboard|SetValueasVector (AI|GetBlackboard self) (Utilities|Name|MakeLiteralName "CoffeePuff_SpawnLocation") _projectedlocation)
    (Utilities|Array|Add (Variables|Default|GetAllKnownValidLocations) _projectedlocation)))

(fn UpdateLastKnownValidTargetLocation
  (params (Location :type Vector))
  (bind _lknavprobe_extentz (Variables|Default|GetLKNavProbe_ExtentZ))
  (bind _returnvalue (ScalableFloat|GetValueatLevel _lknavprobe_extentz))
  (bind _lknavprobe_extenty (Variables|Default|GetLKNavProbe_ExtentY))
  (bind _returnvalue_2 (ScalableFloat|GetValueatLevel _lknavprobe_extenty))
  (bind _lknavprobe_extentx (Variables|Default|GetLKNavProbe_ExtentX))
  (bind _returnvalue_3 (ScalableFloat|GetValueatLevel _lknavprobe_extentx))
  (bind _self self)
  (bind _returnvalue_4 (Fortnite|AI|GetNavigationDataforActor _self))
  (bind (_projectedlocation _returnvalue_5) (AI|Navigation|ProjectPointtoNavigation Location _returnvalue_4 "/Game/Athena/AI/Phoebe/NavFilter/NavFilter_Phoebe.NavFilter_Phoebe_C" (Math|Vector|MakeVector _returnvalue_3 _returnvalue_2 _returnvalue)))
  (if (and _returnvalue_5 (not (AI|Components|NavMovement|IsFalling (Variables|Character|GetCharacterMovement))))
    (AI|Components|Blackboard|SetValueasVector (AI|GetBlackboard self) (Utilities|Name|MakeLiteralName "CoffeePuff_TargetLastKnownLocation") _projectedlocation)
    (Utilities|Array|Add (Variables|Default|GetAllKnownValidLocations) _projectedlocation)
    (Variables|Default|SetHaveValidLastKnownTargetLocation true)
    (if (and (> (Utilities|Array|Length (Variables|Default|GetAllKnownValidLocations)) (ScalableFloat|GetValueatLevel (Variables|Default|GetMaxKnownValidLocations))) (> (Utilities|Array|Length (Variables|Default|GetAllKnownValidLocations)) 1))
      (Utilities|Array|RemoveIndex (Variables|Default|GetAllKnownValidLocations)))))

(fn ConstructionScript
  (|CallParentFunction))
